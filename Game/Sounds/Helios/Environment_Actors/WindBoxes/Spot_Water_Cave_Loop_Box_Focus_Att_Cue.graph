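# Looping water cave ambience SoundCue for a box-shaped spot emitter, reusing the Cattus Cave water loop in the ambient sound class.
# Two nested, endlessly looping Envelopers with different periods vary pitch and volume slowly, then feed a Mixer that boosts the signal.
# Overridden attenuation uses a tall box with listener focus, lowering volume when out of focus, plus a distance-scaled send to distant reverb.
# Overridden concurrency caps it at two voices.
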SoundCue /Game/Sounds/Helios/Environment_Actors/WindBoxes/Spot_Water_Cave_Loop_Box_Focus_Att_Cue
root=Mixer_0
nodes [
 {
  id=Enveloper_4
  children [WavePlayer_0]
  properties {
   bLoop=true
   bLoopIndefinitely=true
   loopEnd=30
   pitchCurve {editorCurveData {keys [
      {arriveTangent=-0.0013, interpMode=RCIM_Cubic, leaveTangent=-0.0013, time=0}
      <at 15 0.98 RCIM_Cubic>
      {arriveTangent=0.0013, interpMode=RCIM_Cubic, leaveTangent=0.0013, time=30}
     ]}}
   volumeCurve {editorCurveData {keys [<at 0 1>]}}
  }
 }
 {
  id=Enveloper_5
  children [Enveloper_4]
  properties {
   bLoop=true
   bLoopIndefinitely=true
   loopEnd=20
   pitchCurve {editorCurveData {keys [{time=0}]}}
   volumeCurve {editorCurveData {keys [
      {arriveTangent=-0.01, interpMode=RCIM_Cubic, time=0, value=1}
      <at 10 1.1 RCIM_Cubic>
      {interpMode=RCIM_Cubic, leaveTangent=0.01, time=20, value=1}
     ]}}
  }
 }
 {id=Mixer_0, children [Enveloper_5], properties {inputVolume [2]}}
 {id=WavePlayer_0, properties {bLooping=true, soundWaveAssetPtr=/Game/Sounds/Artemis/POI/Cattus_Cave/Ambience_Cattus_Cave_Water_Loop}}
]
properties {
 bOverrideAttenuation=true
 bOverrideConcurrency=true
 maxDistance=1050
 soundClassObject=&Fort_Ambient_Athena
 volumeMultiplier=0.351
 attenuationOverrides {
  attenuationShape=Box
  bEnableFocusInterpolation=true
  bEnableListenerFocus=true
  bEnableSubmixSends=true
  falloffDistance=300
  focusAttackInterpSpeed=0.9
  focusReleaseInterpSpeed=0.9
  nonFocusVolumeAttenuation=0.1
  nonSpatializedRadiusEnd=100
  nonSpatializedRadiusStart=237
  attenuationShapeExtents=<237 237 750>
  submixSendSettings [{
    bEnableHPFCutoff=false
    bEnableLPFCutoff=false
    disableManualSendClamp=false
    hPFCutoff=20
    lPFCutoff=20000
    maxSendDistance=6000
    maxSendLevel=1
    minSendDistance=400
    minSendLevel=0.5
    sendLevel=0.5
    sendLevelControlMethod=Linear
    soundSubmix=&Fortnite_Reverb_Distant
   }]
 }
 concurrencyOverrides {maxCount=2, voiceStealReleaseTime=2}
}
